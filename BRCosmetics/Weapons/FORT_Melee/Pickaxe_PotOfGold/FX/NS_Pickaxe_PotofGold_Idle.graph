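# Idle Niagara effect for the Pot of Gold pickaxe, with a rainbow mesh and gold and green shimmers.
# Rainbow bursts a persistent mesh that fades in with system age.
# Shimmers_Packed_Local continuously spawns drifting sprites around the pickaxe; Shimmers_Packed_World spawns sprites and falling, shrinking meshes as it moves, inheriting capped velocity.
# Both shimmer emitters pack several looks into one emitter, switching values by SpawnGroup and a derived VisibilityTag.
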NiagaraSystem /BRCosmetics/Weapons/FORT_Melee/Pickaxe_PotOfGold/FX/NS_Pickaxe_PotofGold_Idle
dependencies {
 usedDataInterfaces [&NiagaraDataInterfaceCurve, &NiagaraDataInterfaceVectorCurve]
 usedDynamicInputs []
 usedModules [&SystemState]
}
emitters {
 Rainbow {
  emitterUpdateScript [
   &EmitterState(
    Inactive Response=::NewEnumerator1
    Life Cycle Mode=::NewEnumerator1
    Loop Behavior=::NewEnumerator1
    Loop Duration Mode=::NewEnumerator1
    UseLoopDelay=false
   )
   &SpawnBurst_Instantaneous(Use Spawn Probability=false, UseLoopCountLimit=false)
  ]
  particleSpawnScript [&InitializeParticle(Mesh Scale Mode=::NewEnumerator4, UsePositionOffset=false, Write Material Random=true)]
  particleUpdateScript [
   &ParticleState(Kill Particles When Lifetime Has Elapsed=false)
   &ScaleColor(
    Scale Alpha=&FloatFromCurve(
     FloatCurve=DataInterface() {propertyValues {Curve {keys [<at 0 0 RCIM_Cubic>, {arriveTangentWeight=0.2, interpMode=RCIM_Cubic, tangentMode=RCTM_User, time=0.6, value=1}]}}}
     CurveIndex=$Engine.System.Age
    )
    ScaleA=true
    ScaleRGB=true
   )
  ]
  rendererClasses [&NiagaraMeshRendererProperties]
  renderers [{rendererClass=&NiagaraMeshRendererProperties, rendererIndex=0}]
 }
 Shimmers_Packed_Local {
  emitterUpdateScript [
   &EmitterState(Inactive Response=::NewEnumerator1, Life Cycle Mode=::NewEnumerator1, Loop Duration=1, UseLoopDelay=false)
   &SpawnRate(SpawnRate=5, Use Spawn Probability=false)
   &SpawnRate(SpawnGroup=1, SpawnRate=5, Use Spawn Probability=false) {module=SpawnRate002}
   &SpawnRate(SpawnGroup=2, SpawnRate=50, Use Spawn Probability=false) {module=SpawnRate001}
   &SpawnRate(SpawnGroup=3, SpawnRate=25, Use Spawn Probability=false) {module=SpawnRate003}
  ]
  particleSpawnScript [
   /SetVariables(Particles.VisibilityTag=&MergeHelper_IntSwitch(0=0, 1=0, 2=1, 3=1, 4=?, 5=?, Index=$Emitter.SpawnGroup, Num=::NewEnumerator2)) {isSetParameters=true}
   /SetVariables(Particles.SpawnGroup=$Emitter.SpawnGroup) {isSetParameters=true}
   &InitializeParticle(
    Color=&MergeHelper_ColorSwitch(
     0=&RandomRangeLinearColor(Minimum=<rgba 20 14.0736 1.1 0.25>, Maximum=<rgba 20 14.0736 1.1 0.5>)
     1=&RandomRangeLinearColor(Minimum=<rgba 20 14.0736 1.1 0.25>, Maximum=<rgba 20 14.0736 1.1 0.5>)
     2=&RandomRangeLinearColor(Minimum=<rgba 2 1.4074 0.11 1>, Maximum=<rgba 20 14.0736 1.1 1>, Color Channel Mode=::NewEnumerator2)
     3=&RandomRangeLinearColor(Minimum=<rgba 0.7294 10 0.355 1>, Maximum=<rgba 1.4587 20 0.71 1>, Color Channel Mode=::NewEnumerator2)
     Index=$Particles.SpawnGroup
     Num=::NewEnumerator2
    )
    Lifetime=&MergeHelper_FloatSwitch(
     0=&RandomRangeFloat(Minimum=1, Maximum=1.5, Evaluation Type=::NewEnumerator1)
     1=&RandomRangeFloat(Minimum=0.5, Maximum=1.5, Evaluation Type=::NewEnumerator1)
     2=?
     3=?
     Index=$Particles.VisibilityTag
     Num=::NewEnumerator0
    )
    Sprite Rotation Mode=::NewEnumerator2
    Sprite Size=&Vector2DFromFloat(Value=&MergeHelper_FloatSwitch(
      0=&RandomRangeFloat(Minimum=1, Maximum=3, Evaluation Type=::NewEnumerator1)
      1=&RandomRangeFloat(Minimum=1, Maximum=5, Evaluation Type=::NewEnumerator1)
      2=&RandomRangeFloat(Minimum=1, Maximum=3, Evaluation Type=::NewEnumerator1)
      3=&RandomRangeFloat(Minimum=1, Maximum=3, Evaluation Type=::NewEnumerator1)
      Index=$Particles.SpawnGroup
      Num=::NewEnumerator2
     ))
    Sprite Size Mode=::NewEnumerator4
    UsePositionOffset=false
    Write Material Random=true
   )
   &ShapeLocation(
    Apply To Particle Position=::NewEnumerator1
    Cylinder Height=&MergeHelper_FloatSwitch(0=5, 1=15, 2=10, 3=10, Index=$Particles.SpawnGroup, Num=::NewEnumerator2)
    Cylinder Height Midpoint:NiagaraFloat=&MergeHelper_FloatSwitch(0=0, 1=0.5, 2=0.5, 3=0.5, Index=$Particles.SpawnGroup, Num=::NewEnumerator2)
    Cylinder Radius=&MergeHelper_FloatSwitch(0=8, 1=12, 2=40, 3=40, Index=$Particles.SpawnGroup, Num=::NewEnumerator2)
    MaskSpawnGroup=false
    Non Uniform Scale=&MaskVector(Mask Source=1, Mask Parameter=$Particles.VisibilityTag, Mask=0, Masked Vector=<1 1 0.5>, Passthrough Vector=<1 -1 1>)
    Offset=&MergeHelper_VectorSwitch(0=<-25 -2 38>, 1=<8 0 35>, 2=<5 0 22>, 3=<5 0 22>, Index=$Particles.SpawnGroup, Num=::NewEnumerator2)
    Offset Coordinate Space=::Local
    Offset Mode=::NewEnumerator0
    Override Seed=false
    Rotation Mode=::NewEnumerator2
    Shape Primitive=::NewEnumerator1
    Surface Only=false
    UseOwnerScale=true
    Yaw / Pitch / Roll=&MaskVector(Mask Source=1, Mask Parameter=$Particles.VisibilityTag, Mask=0, Masked Vector=<0 0 0>, Passthrough Vector=<0 0 0.25>)
   ) {module=ShapeLocation001}
   &ShapeLocation(
    Apply To Particle Position=::NewEnumerator1
    Cylinder Height=&MergeHelper_FloatSwitch(0=5, 1=15, 2=10, 3=10, Index=$Particles.SpawnGroup, Num=::NewEnumerator2)
    Cylinder Height Midpoint:NiagaraFloat=&MergeHelper_FloatSwitch(0=0, 1=0.5, 2=0.5, 3=0.5, Index=$Particles.SpawnGroup, Num=::NewEnumerator2)
    Cylinder Radius=&MergeHelper_FloatSwitch(0=8, 1=12, 2=40, 3=40, Index=$Particles.SpawnGroup, Num=::NewEnumerator2)
    Hemicircle Y=true
    MaskSpawnGroup=false
    Non Uniform Scale=&MaskVector(Mask Source=1, Mask Parameter=$Particles.VisibilityTag, Mask=0, Masked Vector=<1 1 0.5>, Passthrough Vector=<1 -1 1>)
    Offset=&MergeHelper_VectorSwitch(0=<-25 -2 38>, 1=<8 0 35>, 2=<5 0 22>, 3=<5 0 22>, Index=$Particles.SpawnGroup, Num=::NewEnumerator2)
    Offset Coordinate Space=::Local
    Offset Mode=::NewEnumerator0
    Override Seed=false
    Rotation Mode=::NewEnumerator2
    Shape Primitive=::NewEnumerator1
    Surface Only=false
    UseOwnerScale=true
    Yaw / Pitch / Roll=&MaskVector(Mask Source=1, Mask Parameter=$Particles.VisibilityTag, Mask=0, Masked Vector=<0 0 0>, Passthrough Vector=<0 0 0.25>)
   )
   /SetVariables(Particles.Position=&LerpPosition(
     A=$Particles.ShapeLocation001.ShapePosition
     B=$Particles.ShapeLocation.ShapePosition
     Alpha=&MakeFloatFromInt(Int=$Particles.VisibilityTag)
    )) {isSetParameters=true}
   &AddVelocity(
    Module.Use Falloff Distance=false
    Origin Offset=<8 0 35>
    Override Seed=false
    Velocity Mode=::NewEnumerator1
    Velocity Speed=&MergeHelper_FloatSwitch(0=0, 1=3, 2=?, 3=?, Index=$Particles.SpawnGroup, Num=::NewEnumerator0)
   )
  ]
  particleUpdateScript [
   &ParticleState()
   &ScaleSpriteSize(
    Scale Sprite Size Mode=::NewEnumerator0
    Uniform Scale Factor=&MergeHelper_FloatSwitch(
     0=1
     1=&FloatFromCurve(FloatCurve=DataInterface() {propertyValues {Curve {keys [
          {arriveTangent=11.4692, interpMode=RCIM_Cubic, leaveTangent=11.4692, tangentMode=RCTM_User, time=0, value=0}
          <at 0.25 1 RCIM_Cubic>
          {arriveTangent=-4.0234, interpMode=RCIM_Cubic, leaveTangent=-4.0234, tangentMode=RCTM_User, time=1, value=0}
         ]}}})
     2=?
     3=?
     Index=$Particles.VisibilityTag
     Num=::NewEnumerator0
    )
   )
   &SpriteRotationRate(Rotation Rate=&MergeHelper_FloatSwitch(
     0=0
     1=&RandomRangeFloat(Minimum=-0.125, Maximum=0.125)
     2=&RandomRangeFloat(Minimum=-0.125, Maximum=0.25)
     3=&RandomRangeFloat(Minimum=-0.125, Maximum=0.25)
     Index=$Particles.SpawnGroup
     Num=::NewEnumerator2
    ))
   &ScaleColor(
    Scale Alpha=&MergeHelper_FloatSwitch(
     0=&FloatFromCurve(FloatCurve=DataInterface() {propertyValues {Curve {keys [<at 0 1>,<at 1 0>]}}})
     1=&FloatFromCurve(FloatCurve=DataInterface() {propertyValues {Curve {keys [<at 0 0>,<at 0.5 1>,<at 1 0>]}}})
     2=?
     3=?
     Index=$Particles.VisibilityTag
     Num=::NewEnumerator0
    )
    Scale RGB=&MergeHelper_VectorSwitch(
     0=&VectorFromCurve(VectorCurve=DataInterface() {propertyValues {XCurve {keys [<at 0 100>,<at 0.1 1>]}, YCurve {keys [<at 0 100>,<at 0.1 1>]}, ZCurve {keys [<at 0 100>,<at 0.1 1>]}}})
     1=<1 1 1>
     2=?
     3=?
     Index=$Particles.VisibilityTag
     Num=::NewEnumerator0
    )
    ScaleA=true
    ScaleRGB=true
   )
   &AccelerationForce(
    Acceleration=&MergeHelper_VectorSwitch(0=<0 0 -2>, 1=<0 0 -1>, 2=<0 0 -25>, 3=<0 0 -25>, Index=$Particles.SpawnGroup, Num=::NewEnumerator2)
    Coordinate Space=::Local
   )
   &SolveForcesAndVelocity(Clamp Velocity=false, Limit Acceleration=false)
  ]
  rendererClasses [&NiagaraSpriteRendererProperties]
  renderers [
   {rendererClass=&NiagaraSpriteRendererProperties}
   {rendererClass=&NiagaraSpriteRendererProperties, rendererIndex=1}
  ]
 }
 Shimmers_Packed_World {
  emitterUpdateScript [
   &EmitterState(Inactive Response=::NewEnumerator1, Life Cycle Mode=::NewEnumerator1, Loop Duration=1, UseLoopDelay=false)
   &SpawnPerUnit(
    Module.Use Max Movement Threshold=false
    Module.Use Movement Tolerance=false
    Spawn Spacing:NiagaraFloat=&RandomRangeFloat(Minimum=50, Maximum=25, Evaluation Type=::NewEnumerator1)
    Use Spawn Probability=false
   )
   &SpawnPerUnit(
    Module.Use Max Movement Threshold=false
    Module.Use Movement Tolerance=false
    Spawn Group=1
    Spawn Spacing:NiagaraFloat=&RandomRangeFloat(Minimum=50, Maximum=25, Evaluation Type=::NewEnumerator1)
    Use Spawn Probability=false
   ) {module=SpawnPerUnit001}
   &SpawnPerUnit(Module.Use Max Movement Threshold=false, Module.Use Movement Tolerance=false, Spawn Group=2, Spawn Spacing=200, Use Spawn Probability=false) {module=SpawnPerUnit002}
  ]
  particleSpawnScript [
   /SetVariables(Particles.VisibilityTag=&MergeHelper_IntSwitch(0=0, 1=0, 2=1, 3=?, 4=?, 5=?, Index=$Emitter.SpawnGroup, Num=::NewEnumerator1)) {isSetParameters=true}
   /SetVariables(Particles.SpawnGroup=$Emitter.SpawnGroup) {isSetParameters=true}
   &InitializeParticle(
    Color=&MergeHelper_ColorSwitch(
     0=&RandomRangeLinearColor(Minimum=<rgba 2 1.4074 0.11 1>, Maximum=<rgba 20 14.0736 1.1 1>)
     1=&RandomRangeLinearColor(Minimum=<rgba 0.7294 10 0.355 1>, Maximum=<rgba 1.4587 20 0.71 1>)
     2=<rgba 0.5 0.5 0.5 1>
     3=?
     Index=$Particles.SpawnGroup
     Num=::NewEnumerator1
    )
    Lifetime=&MergeHelper_FloatSwitch(
     0=&RandomRangeFloat(Minimum=0.5, Maximum=1.5, Evaluation Type=::NewEnumerator1)
     1=0.75
     2=?
     3=?
     Index=$Particles.VisibilityTag
     Num=::NewEnumerator0
    )
    Mesh Scale=<0.45 0.45 0.45>
    Mesh Scale Mode=::NewEnumerator4
    Position Offset=&MergeHelper_VectorSwitch(
     0=&RandomRangeVector(Minimum=<-8 -4 35>, Maximum=<22 4 55>, Evaluation Type=::NewEnumerator1)
     1=<-25 0 38>
     2=?
     3=?
     Index=$Particles.VisibilityTag
     Num=::NewEnumerator0
    )
    Position Offset Coordinate Space=::Local
    Sprite Rotation Mode=::NewEnumerator2
    Sprite Size=&Vector2DFromFloat(Value=&RandomRangeFloat(Minimum=1, Maximum=3, Evaluation Type=::NewEnumerator1))
    Sprite Size Mode=::NewEnumerator4
    UsePositionOffset=true
    Write Material Random=true
   )
   &InitialMeshOrientation(Use Rotation Vector=false)
   &InheritVelocity(Inherited Velocity Amount Scale=<0.98 0.98 0.98>, Inherited Velocity Speed Limit=15, Speed Limit=true, UseSpeedThreshold=false)
  ]
  particleUpdateScript [
   &ParticleState()
   &SpriteRotationRate(Rotation Rate=&RandomRangeFloat(Minimum=-0.125, Maximum=0.25))
   &ScaleMeshSize(Scale Factor=&VectorFromFloat(Value=&FloatFromCurve(FloatCurve=DataInterface() {propertyValues {Curve {keys [<at 0 1 RCIM_Cubic>, {arriveTangent=-3.1038, interpMode=RCIM_Cubic, leaveTangent=-3.1038, tangentMode=RCTM_User, time=1, value=0}]}}})))
   &UpdateMeshOrientation()
   &AccelerationForce(Acceleration=&MergeHelper_VectorSwitch(0=<0 0 -25>, 1=<0 0 -250>, 2=?, 3=?, Index=$Particles.VisibilityTag, Num=::NewEnumerator0))
   &SolveForcesAndVelocity(Clamp Velocity=false, Limit Acceleration=false)
   &ScaleColor(
    Scale Alpha=&MergeHelper_FloatSwitch(
     0=&FloatFromCurve(FloatCurve=DataInterface() {propertyValues {Curve {keys [<at 0 0>,<at 0.5 1>,<at 1 0>]}}})
     1=1
     2=?
     3=?
     Index=$Particles.VisibilityTag
     Num=::NewEnumerator0
    )
    ScaleA=true
    ScaleRGB=true
   )
  ]
  rendererClasses [&NiagaraSpriteRendererProperties, &NiagaraMeshRendererProperties]
  renderers [
   {rendererClass=&NiagaraSpriteRendererProperties}
   {rendererClass=&NiagaraMeshRendererProperties, rendererIndex=1}
  ]
 }
}
summary {
 systemName=NS_Pickaxe_PotofGold_Idle
 emitters [
  {
   emitterName=Shimmers_Packed_World
   rendererClasses [&NiagaraSpriteRendererProperties, &NiagaraMeshRendererProperties]
  }
  {emitterName=Rainbow, rendererClasses [&NiagaraMeshRendererProperties]}
  {emitterName=Shimmers_Packed_Local, rendererClasses [&NiagaraSpriteRendererProperties]}
 ]
 userVariables []
}
